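# Physics asset for the body mesh of the male medium Biker 01 player cosmetic.
# It has capsule bodies on only the pelvis, spine_01, right thigh and right calf, with none for the left leg, arms or head.
# Constraints link the right thigh and spine_01 to the pelvis and the right calf to the right thigh.
# All constraint swing and twist motions are Free, so their angle limits are not enforced.
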PhysicsAsset /BRCosmetics/Characters/Player/Male/Medium/Bodies/M_MED_Biker_01/Meshes/M_MED_Biker_01_Body_Physics
bodies {
 calf_r [{shapeType=Capsule, length=36.0552, radius=9.6992, center=<17.4338 -1.5804 1.6125>, rotation=<rot 0 88.3896 -89.0115>}]
 pelvis [{shapeType=Capsule, length=6.6852, radius=9.75, center=<-6.6428 -0.3505 2.5919>}]
 spine_01 [{shapeType=Capsule, length=27.5592, radius=11.008, center=<5.2708 0.0806 1.2754>, rotation=<rot 88.1492 -88.1619 -90.0104>}]
 thigh_r [{shapeType=Capsule, length=34.359, radius=9.2342, center=<21.59 -2.3126 -0.4324>, rotation=<rot -84.9894 3.0861 -3.3697>}]
}
constraints [
 {
  bone1Name=thigh_r
  bone2Name=pelvis
  swing1LimitDegrees=45
  swing1Motion=Free
  swing2LimitDegrees=45
  swing2Motion=Free
  twistLimitDegrees=45
  twistMotion=Free
 }
 {
  bone1Name=spine_01
  bone2Name=pelvis
  swing1LimitDegrees=45
  swing1Motion=Free
  swing2LimitDegrees=45
  swing2Motion=Free
  twistLimitDegrees=45
  twistMotion=Free
 }
 {
  bone1Name=calf_r
  bone2Name=thigh_r
  swing1LimitDegrees=45
  swing1Motion=Free
  swing2LimitDegrees=45
  swing2Motion=Free
  twistLimitDegrees=45
  twistMotion=Free
 }
]
